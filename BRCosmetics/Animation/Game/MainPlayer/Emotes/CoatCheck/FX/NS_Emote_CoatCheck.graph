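# Niagara effect for the CoatCheck emote that shows a pair of glowing wings with a flash ring, pulsing glow and sparks.
# Wings are two mirrored mesh particles split by PartitionParticles, with timeline-indexed rotation, stretch-in and fade-out material curves, sprites and lights.
# Emitters anchor to System.SourcePosition, with Glow and a GPU spark burst timed to System.UnfoldTime and sparks thrown in two opposing cones.
# The center-embers and debug emitters are disabled.
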NiagaraSystem /BRCosmetics/Animation/Game/MainPlayer/Emotes/CoatCheck/FX/NS_Emote_CoatCheck
dependencies {
 usedDataInterfaces [
  &NiagaraDataInterfaceColorCurve
  &NiagaraDataInterfaceCurve
  &NiagaraDataInterfaceArrayFloat
  &NiagaraDataInterfaceArrayFloat3
 ]
 usedDynamicInputs []
 usedModules [
  &self:SystemSpawnScript.SystemScriptSource.SystemScriptGraph.NiagaraNodeAssignment_2.SetVariables_#1
  &SystemState
  &Timeline
  &self:SystemSpawnScript.SystemScriptSource.SystemScriptGraph.NiagaraNodeAssignment_0.SetVariables_#2
  &SampleSkeletalMesh
  &self:SystemSpawnScript.SystemScriptSource.SystemScriptGraph.NiagaraNodeAssignment_1.SetVariables_#3
 ]
}
emitters {
 FN_AOE_Embers_Center {
  bEnabled=false
  simTarget=GPUComputeSim
  emitterUpdateScript [
   &EmitterState()
   &SpawnRate(
    SpawnRate=&FloatFromCurve(FloatCurve=DataInterface() {propertyValues {Curve {keys [<at 0 100>,<at 0.15 100>,<at 0.2 0>]}}}, CurveIndex=$Emitter.NormalizedLoopAge)
    Use Spawn Probability=false
   ) {enabled=false}
   &SpawnBurst_Instantaneous(Spawn Count=&RandomRangeInt(Minimum=40, Maximum=90), Use Spawn Probability=false)
  ]
  particleSpawnScript [
   /Niagara/Modules/Spawn/Initialization/InitializeParticle(
    Color=<rgba 10 6.0112 3.2412 1>
    Lifetime=&UniformRangedFloat(Minimum=0.25, Maximum=2)
    Module.Write Color=true
    Module.Write Lifetime=true
    Module.Write Mass=false
    Module.Write Material Random=true
    Module.Write Position=false
    Module.Write Scale=false
    Module.Write SpriteRotation=true
    Module.Write SpriteSize=true
    Position=&SimulationPosition()
    Sprite Rotation=&UniformRangedFloat(Minimum=-180, Maximum=180)
    Sprite Size=&Vector2DFromFloat(Value=&UniformRangedFloat(Minimum=1, Maximum=2))
   )
   &SkeletalMeshLocation(Bone Sampling Mode=::NewEnumerator4, MaskSpawnGroup=false, Sampled Position Offset=<0 0 150>, Skeletal Mesh=$User.SkeletalMesh) {enabled=false, module=SkeletalMeshLocation001}
   &AddVelocity(Velocity=&MakeVector(X=0, Y=0, Z=&UniformRangedFloat(Minimum=25, Maximum=75))) {enabled=false}
   /Niagara/Modules/Spawn/Location/SphereLocation(
    MaskSpawnGroup=false
    Module.Override Randomness=false
    Offset=<0 0 -100>
    Sphere Origin:NiagaraPosition=$System.SourcePosition
    Sphere Radius=20
    Surface Only=false
   )
   &AddVelocityFromPoint(
    Module.Use Falloff Distance=false
    Velocity Strength=&Multiply_Float(A=&RandomRangeFloat(Minimum=200, Maximum=600), B=1.4)
   )
   &ApplyInitialForces()
  ]
  particleUpdateScript [
   &ParticleState()
   /Niagara/Modules/Update/Forces/CurlNoiseForce(
    Noise Frequency=0.0125
    Noise Strength=&FloatFromCurve(FloatCurve=DataInterface() {propertyValues {Curve {keys [<at 0 1>,<at 1 0>]}}}, Scale Curve=1000)
    Pan Noise=true
    Pan Noise Field=<0 0 1>
   )
   &AccelerationForce(Acceleration=<0 0 250>)
   &Drag(Drag=8, Use Linear Drag=true, Use Rotational Drag=false)
   &ScaleColor(
    Scale Alpha=&FloatFromCurve(FloatCurve=DataInterface() {propertyValues {Curve {keys [
         {arriveTangent=11.9433, interpMode=RCIM_Cubic, leaveTangent=11.9433, tangentMode=RCTM_User, time=0, value=0}
         <at 0.2 1 RCIM_Cubic>
         <at 1 0 RCIM_Cubic>
        ]}}})
    Scale RGB=&MakeVectorFromLinearColorRGB(Color=&ColorFromCurve(ColorCurve=DataInterface() {propertyValues {
        AlphaCurve {keys [<at 0 1 RCIM_Cubic>]}
        BlueCurve {keys [<at 0 0.5069 RCIM_Cubic>,<at 1 0 RCIM_Cubic>]}
        GreenCurve {keys [<at 0 1.6422 RCIM_Cubic>,<at 1 0.06 RCIM_Cubic>]}
        RedCurve {keys [<at 0 4 RCIM_Cubic>,<at 1 1 RCIM_Cubic>]}
       }}))
    ScaleA=true
    ScaleRGB=true
   )
   &DynamicMaterialParameters(
    Index 0 Param 1=&FloatFromCurve(FloatCurve=DataInterface() {propertyValues {Curve {keys [<at 0 0>,<at 1 1>]}}})
    Index 0 Param 2=1
    Index 0 Param 3=1
    Index 0 Param 4=1
    Module.Param0WriteEnabled=true
    Module.Param1WriteEnabled=true
    Module.Param2WriteEnabled=true
    Module.Param3WriteEnabled=true
   )
   &ScaleSpriteSize(Scale Factor=&Vector2DFromFloat(Value=&FloatFromCurve(FloatCurve=DataInterface() {propertyValues {Curve {keys [
          {arriveTangent=9.1531, interpMode=RCIM_Cubic, leaveTangent=9.1531, tangentMode=RCTM_User, time=0, value=0.02}
          {arriveTangent=0.48, interpMode=RCIM_Cubic, leaveTangent=0.48, time=0.18, value=0.99}
          <at 1 0.5 RCIM_Cubic>
         ]}}})))
   &SolveForcesAndVelocity(Clamp Velocity=false, Limit Acceleration=false)
   &ScaleSpriteSizeBySpeed(Max Scale Factor=<1 4>, Min Scale Factor=<1 2>)
  ]
  rendererClasses [&NiagaraSpriteRendererProperties]
  renderers [{rendererClass=&NiagaraSpriteRendererProperties}]
 }
 FN_Debug {
  bEnabled=false
  emitterUpdateScript [
   &EmitterState()
   &SpawnBurst_Instantaneous(Spawn Count=1, Use Spawn Probability=false)
  ]
  particleSpawnScript [&InitializeParticle(
    Lifetime=1568.3982
    Module.Write Color=true
    Module.Write Lifetime=true
    Module.Write Mass=true
    Module.Write Material Random=true
    Module.Write Position=false
    Module.Write Scale=false
    Module.Write SpriteRotation=false
    Module.Write SpriteSize=true
    Position=&SimulationPosition()
    Sprite Size=<100 100>
   )]
  particleUpdateScript [
   &ParticleState()
   /SetVariables(Particles.Debug:NiagaraFloat=Linked() {linkedVariable {name=NPC.FortniteNPC.TimeOfDay_0-1, type=NiagaraFloat}}) {isSetParameters=true}
   &DynamicMaterialParameters(
    Index 0 Param 1=$Particles.Debug
    Module.Param0WriteEnabled=true
    Module.Param1WriteEnabled=true
    Module.Param2WriteEnabled=true
    Module.Param3WriteEnabled=true
   )
  ]
  rendererClasses [&NiagaraSpriteRendererProperties]
  renderers [{rendererClass=&NiagaraSpriteRendererProperties}]
 }
 Glow {
  emitterSpawnScript [/SetVariables(Emitter.GlowDepthFade=50) {isSetParameters=true}]
  emitterUpdateScript [
   &EmitterState()
   &SpawnBurst_Instantaneous(Spawn Count=1, Spawn Time=$System.UnfoldTime, Use Spawn Probability=false)
  ]
  particleSpawnScript [
   /Niagara/Modules/Spawn/Initialization/V2/InitializeParticle(
    Color=<rgba 0.7148 0.6797 0.75 1>
    Lifetime=1.2
    Sprite Rotation Mode=::NewEnumerator0
    Sprite Size Mode=::NewEnumerator3
    Uniform Sprite Size=250
    UsePositionOffset=true
    Write Material Random=true
   )
   /SetVariables(Particles.FlareColor=<rgba 1 1 1 25>) {enabled=false, isSetParameters=true}
   /SetVariables(Particles.FlareSize:Vector2f=<40 10>) {enabled=false, isSetParameters=true}
   /SetVariables(Particles.FlareAlignement:Vector3f=$Engine.Owner.SystemZAxis) {enabled=false, isSetParameters=true}
   /SetVariables(Particles.FlareFacing:Vector3f=$Engine.Owner.SystemYAxis) {enabled=false, isSetParameters=true}
   &CameraOffset(Camera Offset Amount=20) {enabled=false}
  ]
  particleUpdateScript [
   &ParticleState()
   /SetVariables(Particles.Position=&TransformPosition(Position=$System.SourcePosition, Source Space=::World, Destination Space=::Simulation)) {isSetParameters=true}
   &SpriteRotationRate(Rotation Rate=-30)
   &ScaleColor(
    Scale Alpha=&Multiply_Float(
     A=&Waveform(Global Amplitude Scale=0.5, "Waveform [1]"=::NewEnumerator12, "[1] Frequency"=8, "[1] Amplitude Min/Max"=<0.2 1>)
     B=&FloatFromCurve(FloatCurve=DataInterface() {propertyValues {Curve {keys [
          {arriveTangent=9.0725, interpMode=RCIM_Cubic, time=0, value=0}
          {arriveTangent=1.4286, interpMode=RCIM_Cubic, leaveTangent=1.4286, time=0.1408, value=0.9847}
          {arriveTangent=-1.146, interpMode=RCIM_Cubic, leaveTangent=-1.146, time=0.7, value=1}
          {interpMode=RCIM_Cubic, leaveTangent=-0.0405, time=1, value=0}
         ]}}})
    )
    Scale RGB=&MakeVectorFromLinearColorRGB(Color=&MultiplyLinearColor(
      Linear Color=<rgba 5 0 2.7134 1>
      Scale Factor=&MakeVector4_FromFloat(Float=&Waveform(Global Amplitude Scale=0.5, "Waveform [1]"=::NewEnumerator12, "[1] Frequency"=7, "[1] Amplitude Min/Max"=<1 5>))
     ))
    ScaleA=true
    ScaleRGB=true
   )
   &ScaleColor(
    Color Value To Scale=$StackContext.Color
    Scale Alpha=&Waveform(
     Waveform Count=::NewEnumerator1
     "Waveform [1]"=::NewEnumerator12
     "[1] Frequency"=30
     "[1] Amplitude Min/Max"=<0 1>
     "[2] Average"=false
     "Waveform [2]"=::NewEnumerator12
     "[2] Input Position"=0
     "[2] Phase"=$Emitter.Age
     "[2] Phase Scale"=1
     "[2] Compound Frequency"=2
     "[2] Frequency"=10
     "[2] Amplitude Min/Max"=<0 1>
     "[2] Scale"=5
     "[2] Bias"=0
     "[2] Exponent"=1
     "[2] Pulse Duration"=0.1
     "[2] Clip Triangle Min/Max"=<0 0>
     "[2] Random Seed"=0
    )
    ScaleA=true
    ScaleRGB=false
   ) {enabled=false, module=ScaleColor001}
   &DynamicMaterialParameters(
    Index 0 Param 1=1
    Module.Param0WriteEnabled=true
    Module.Param0WriteEnabled1=true
    Module.Param1WriteEnabled=true
    Module.Param1WriteEnabled1=true
    Module.Param2WriteEnabled=true
    Module.Param2WriteEnabled1=true
    Module.Param3WriteEnabled=true
    Module.Param3WriteEnabled1=true
    Write Parameter Index 1=true
   )
   &ScaleSpriteSize(
    Uniform Curve Scale=0.5
    Uniform Curve Sprite Scale=DataInterface() {propertyValues {Curve {keys [{tangentMode=RCTM_User, time=0, value=0}, <at 0.1 1>, {tangentMode=RCTM_User, time=1, value=1}]}}}
   )
   &SkeletalMeshLocation(Bone / Socket Index=1, Bone Sampling Mode=::NewEnumerator4, MaskSpawnGroup=false, Skeletal Mesh=$User.SkeletalMesh) {enabled=false}
   &FN_Gen5_Light(
    Attenutation:NiagaraFloat=&FloatFromCurve(
     FloatCurve=DataInterface() {propertyValues {Curve {keys [
         {arriveTangent=9.0725, interpMode=RCIM_Cubic, leaveTangent=9.0725, tangentMode=RCTM_User, time=0, value=0}
         {arriveTangent=0.173, interpMode=RCIM_Cubic, leaveTangent=0.173, tangentMode=RCTM_User, time=0.1408, value=0.9847}
         {arriveTangent=-2.2361, interpMode=RCIM_Cubic, leaveTangent=-2.2361, tangentMode=RCTM_User, time=1, value=0.25}
        ]}}}
     Scale Curve=400
    )
    Exponent=2.5
    Intensity=40000
   ) {enabled=false}
   &FN_Light_Attributes(
    Color=<rgba 1 0.1491 0 1>
    Create Color Attribute=true
    Create Position Attribute=true
    Create Radius Attribute=true
    Enable Particle Light Enable=true
    Position=&AddVectorToPosition(Position=$Particles.Position, Vector=&TransformVector(Vector=<0 -10 0>, Destination Space=::Simulation))
    Scale Color=&FloatFromCurve(
     FloatCurve=DataInterface() {propertyValues {Curve {keys [
         {arriveTangent=9.0725, interpMode=RCIM_Cubic, leaveTangent=9.0725, tangentMode=RCTM_User, time=0, value=0}
         {arriveTangent=-2.5249, interpMode=RCIM_Cubic, leaveTangent=-2.5249, tangentMode=RCTM_User, time=0.1408, value=0.9847}
         {arriveTangent=1.4346, interpMode=RCIM_Cubic, leaveTangent=1.4346, time=0.4, value=0.5}
         {arriveTangent=1.25, interpMode=RCIM_Cubic, leaveTangent=1.25, time=0.5, value=1.5}
         {arriveTangent=-1.6667, interpMode=RCIM_Cubic, leaveTangent=-1.6667, time=0.6, value=0.75}
         {arriveTangent=-1.875, interpMode=RCIM_Cubic, leaveTangent=-1.875, time=0.8, value=1}
         {arriveTangent=-0.0405, interpMode=RCIM_Cubic, leaveTangent=-0.0405, tangentMode=RCTM_User, time=1, value=0}
        ]}}}
     Scale Curve=40
    )
   )
   /SetVariables(
    Particles.AccentColor:LinearColor=&MultiplyLinearColor(
     Linear Color=<rgba 1 0 0.6053 1>
     Scale Factor=&MakeFloat4FromVectorAndFloat(
      "Vector (RGB)"=<1 1 1>
      "Float (Alpha)"=&FloatFromCurve(
       FloatCurve=DataInterface() {propertyValues {Curve {keys [
           <at 0 0 RCIM_Cubic>
           {arriveTangent=0.625, interpMode=RCIM_Cubic, leaveTangent=0.625, time=0.3, value=1}
           {arriveTangent=-1.4286, interpMode=RCIM_Cubic, leaveTangent=-1.4286, time=0.8, value=0.5}
           <at 1 0 RCIM_Cubic>
          ]}}}
       Scale Curve=8
      )
     )
    )
    Particles.AccentRotation:NiagaraFloat=&Multiply_Float(A=$System.Age, B=35)
    Particles.AccentSize:Vector2f=<250 250>
   ) {isSetParameters=true}
   /SetVariables(Particles.LightIntensity:NiagaraFloat=&FloatFromCurve(
     FloatCurve=DataInterface() {propertyValues {Curve {keys [
         {arriveTangent=9.0725, interpMode=RCIM_Cubic, leaveTangent=9.0725, tangentMode=RCTM_User, time=0, value=0}
         {arriveTangent=-2.5249, interpMode=RCIM_Cubic, leaveTangent=-2.5249, tangentMode=RCTM_User, time=0.1408, value=0.9847}
         {arriveTangent=-1.146, interpMode=RCIM_Cubic, leaveTangent=-1.146, time=0.8, value=1.075}
         {arriveTangent=-0.0405, interpMode=RCIM_Cubic, leaveTangent=-0.0405, tangentMode=RCTM_User, time=1, value=0}
        ]}}}
     Scale Curve=700
    )) {isSetParameters=true}
  ]
  rendererClasses [
   &NiagaraSpriteRendererProperties
   &NiagaraLightRendererProperties
   &NiagaraMeshRendererProperties
   &NiagaraComponentRendererProperties
  ]
  renderers [
   {rendererClass=&NiagaraSpriteRendererProperties}
   {rendererClass=&NiagaraLightRendererProperties, rendererIndex=1}
   {rendererClass=&NiagaraSpriteRendererProperties, rendererIndex=2}
   {rendererClass=&NiagaraMeshRendererProperties, rendererIndex=3}
   {rendererClass=&NiagaraComponentRendererProperties, rendererIndex=4}
  ]
 }
 Ring {
  emitterSpawnScript [/SetVariables(Emitter.GlowDepthFade=25) {isSetParameters=true}]
  emitterUpdateScript [
   &EmitterState(
    Enable Distance Culling=true
    Life Cycle Mode=::NewEnumerator1
    Loop Behavior=::NewEnumerator1
    Loop Delay=0.185
    Loop Duration=1
    Max Distance Response=::KillImmediately
    MaxDistance=3500
    Module.bUseMaxDistance=true
    Module.bUseMinDistance=false
    Scalability Mode=::NewEnumerator1
    UseLoopDelay=true
   )
   &SpawnBurst_Instantaneous(Spawn Count=1, Use Spawn Probability=false)
  ]
  particleSpawnScript [
   &InitializeParticle(
    Color=<rgba 1 0.9671 0.9772 1>
    Lifetime=0.25
    Position Offset Coordinate Space=::Local
    Sprite Rotation Mode=::NewEnumerator2
    Sprite Size Mode=::NewEnumerator3
    Uniform Sprite Size=450
    UsePositionOffset=true
    Write Material Random=true
   )
   &CameraOffset(Camera Offset Amount=10)
   /SetVariables(Particles.Position=&TransformPosition(Position=$System.SourcePosition, Source Space=::World, Destination Space=::Simulation)) {isSetParameters=true}
  ]
  particleUpdateScript [
   &ParticleState()
   /SetVariables(Particles.Position=&TransformPosition(Position=$System.SourcePosition, Source Space=::World, Destination Space=::Simulation)) {isSetParameters=true}
   &ScaleColor(
    Scale Alpha=&FloatFromCurve(FloatCurve=DataInterface() {propertyValues {Curve {keys [<at 0 0.5>,<at 1 0>]}}})
    Scale RGB=&MakeVectorFromLinearColorRGB(Color=&ColorFromCurve(ColorCurve=DataInterface() {propertyValues {
        AlphaCurve {keys [<at 0 1 RCIM_Cubic>,<at 1 0 RCIM_Cubic>]}
        BlueCurve {keys [<at 0.0117 1>, {arriveTangent=-0.5535, interpMode=RCIM_Cubic, leaveTangent=-0.5535, time=0.0431, value=0}, <at 0.6613 0.6404 RCIM_Cubic>]}
        GreenCurve {keys [
          <at 0.0117 1>
          {arriveTangent=-0.7711, interpMode=RCIM_Cubic, leaveTangent=-0.7711, time=0.0431, value=0.3574}
          <at 0.6613 0.4991 RCIM_Cubic>
         ]}
        RedCurve {keys [<at 0.0117 1>,<at 0.0431 5 RCIM_Cubic>,<at 0.6613 1 RCIM_Cubic>]}
       }}))
    ScaleA=true
    ScaleRGB=true
   )
   &ScaleSpriteSize(Scale Factor=&Vector2DFromFloat(Value=&FloatFromCurve(FloatCurve=DataInterface() {propertyValues {Curve {keys [{arriveTangent=2.7513, interpMode=RCIM_Cubic, leaveTangent=2.7513, tangentMode=RCTM_User, time=0, value=0}, <at 1 1>]}}})))
   /SetVariables(
    Particles.FlareColor=&MultiplyLinearColor(
     Linear Color=&ColorFromCurve(ColorCurve=DataInterface() {propertyValues {
        AlphaCurve {keys [<at 0 1 RCIM_Cubic>,<at 1 0 RCIM_Cubic>]}
        BlueCurve {keys [<at 0 0.019 RCIM_Cubic>,<at 0.9935 0.5036>]}
        GreenCurve {keys [<at 0 0.5673 RCIM_Cubic>,<at 0.9935 0.3015>]}
        RedCurve {keys [<at 0 5 RCIM_Cubic>,<at 0.9935 1>]}
       }})
     Scale Factor=&MakeVector4_FromFloat(Float=1)
    )
    Particles.FlareSize:Vector2f=&Multiply_Vector2D_ByFloat(
     Float=&FloatFromCurve(FloatCurve=DataInterface() {propertyValues {Curve {keys [
          {interpMode=RCIM_Cubic, tangentMode=RCTM_User, time=0, value=0}
          {interpMode=RCIM_Cubic, tangentMode=RCTM_User, time=0.3, value=0.95}
          {tangentMode=RCTM_User, time=1, value=0}
         ]}}})
     Vector2D=$Particles.Initial.SpriteSize
    )
    Particles.FlashCameraOffset=34.9355
   ) {isSetParameters=true}
   &SkeletalMeshLocation(
    Bone / Socket Index=1
    Bone Sampling Mode=::NewEnumerator4
    MaskSpawnGroup=false
    Sampled Position Offset=<0 -30 0>
    Skeletal Mesh=$User.SkeletalMesh
   ) {enabled=false}
  ]
  rendererClasses [&NiagaraSpriteRendererProperties]
  renderers [
   {rendererClass=&NiagaraSpriteRendererProperties}
   {rendererClass=&NiagaraSpriteRendererProperties, rendererIndex=1}
  ]
 }
 Sparks_Wings {
  simTarget=GPUComputeSim
  emitterUpdateScript [
   &EmitterState()
   &SpawnBurst_Instantaneous(Spawn Count=50, Spawn Group=1, Spawn Time=1, Use Spawn Probability=false)
   &SpawnBurst_Instantaneous(Spawn Count=80, Spawn Time=$System.UnfoldTime, Use Spawn Probability=false, UseLoopCountLimit=false) {module=SpawnBurst_Instantaneous001}
   &SpawnRate(
    SpawnRate=&FloatFromCurve(FloatCurve=DataInterface() {propertyValues {Curve {keys [<at 0 100>,<at 0.15 100>,<at 0.2 0>]}}}, CurveIndex=$Emitter.NormalizedLoopAge)
    Use Spawn Probability=false
   ) {enabled=false}
  ]
  particleSpawnScript [
   /Niagara/Modules/Spawn/Initialization/InitializeParticle(
    Color=<rgba 10 6.0112 3.2412 1>
    Lifetime=&UniformRangedFloat(Minimum=0.5, Maximum=1.5)
    Module.Write Color=true
    Module.Write Lifetime=true
    Module.Write Mass=false
    Module.Write Material Random=true
    Module.Write Position=false
    Module.Write Scale=false
    Module.Write SpriteRotation=true
    Module.Write SpriteSize=true
    Position=&SimulationPosition()
    Sprite Rotation=&UniformRangedFloat(Minimum=-180, Maximum=180)
    Sprite Size=&Vector2DFromFloat(Value=&UniformRangedFloat(Minimum=1.25, Maximum=2.25))
   ) {enabled=false}
   /Niagara/Modules/Spawn/Initialization/V2/InitializeParticle(
    Color=<rgba 10 1.0866 0 1>
    Lifetime=&UniformRangedFloat(Minimum=0.5, Maximum=1)
    Sprite Rotation Angle=&UniformRangedFloat(Minimum=-180, Maximum=180)
    Sprite Rotation Mode=::NewEnumerator0
    Sprite Size=&Vector2DFromFloat(Value=&UniformRangedFloat(Minimum=1.25, Maximum=2.25))
    Sprite Size Mode=::NewEnumerator4
    UsePositionOffset=false
    Write Material Random=true
   ) {module=InitializeParticle001}
   &PartitionParticles()
   &SkeletalMeshLocation(Bone / Socket Index=1, Bone Sampling Mode=::NewEnumerator4, MaskSpawnGroup=false, Skeletal Mesh=$User.SkeletalMesh) {enabled=false, module=SkeletalMeshLocation001}
   &AddVelocity(Velocity=&MakeVector(X=0, Y=0, Z=&UniformRangedFloat(Minimum=25, Maximum=75))) {enabled=false}
   &SphereLocation(
    MaskSpawnGroup=false
    Module.Override Randomness=false
    Offset=&MaskVectorBySpawnGroup(Masked Vector=<0 0 0>, Passthrough Vector=<0 0 0>)
    Sphere Origin:NiagaraPosition=$System.SourcePosition
    Sphere Radius=20
    Surface Only=false
   )
   &AddVelocity(
    Cone Angle=30
    Cone Axis=&SelectVectorFromArray(
     Direct Array Index=$Particles.PartitionParticles.Partition
     Vector Selection Array=DataInterface() {propertyValues {FloatData [<1 -0.5 0.4>, <-1 -0.5 0.4>]}}
    )
    Module.UseConeFalloff=false
    Module.UseVelDistribution=false
    Override Seed=false
    Velocity Mode=::NewEnumerator2
    Velocity Speed=&RandomRangeFloat(Minimum=100, Maximum=400)
   ) {module=AddVelocity002}
   &ApplyInitialForces()
   &AddVelocity() {enabled=false, module=AddVelocity001}
   &AddVelocity(
    Velocity=<0 0 -1>
    Velocity Speed Scale=&MaskFloatBySpawnGroup(Masked Float=0, Passthrough Float=&RandomRangeFloat(Minimum=100, Maximum=250))
   ) {module=AddVelocity003}
  ]
  particleUpdateScript [
   &ParticleState()
   &CurlNoiseForce(
    Noise Frequency=0.0125
    Noise Strength=&FloatFromCurve(FloatCurve=DataInterface() {propertyValues {Curve {keys [<at 0 1>,<at 1 0>]}}}, Scale Curve=1000)
    Pan Noise=true
    Pan Noise Field=<3 0 0>
   )
   &AccelerationForce(Acceleration=<0 0 250>)
   &Drag(Drag=2, Use Linear Drag=true, Use Rotational Drag=false)
   &ScaleColor(
    Scale Alpha=&FloatFromCurve(FloatCurve=DataInterface() {propertyValues {Curve {keys [
         {arriveTangent=11.9433, interpMode=RCIM_Cubic, leaveTangent=11.9433, tangentMode=RCTM_User, time=0, value=0}
         <at 0.2 1 RCIM_Cubic>
         <at 1 0 RCIM_Cubic>
        ]}}})
    Scale RGB=&MakeVectorFromLinearColorRGB(Color=&ColorFromCurve(ColorCurve=DataInterface() {propertyValues {
        AlphaCurve {keys [<at 0 1 RCIM_Cubic>]}
        BlueCurve {keys [<at 0 0.2451 RCIM_Cubic>,<at 0.5253 0.0263 RCIM_Cubic>]}
        GreenCurve {keys [<at 0 0.7408 RCIM_Cubic>,<at 0.5253 0 RCIM_Cubic>]}
        RedCurve {keys [<at 0 1 RCIM_Cubic>,<at 0.5253 0.3802 RCIM_Cubic>]}
       }}))
    ScaleA=true
    ScaleRGB=true
   )
   &DynamicMaterialParameters(
    Index 0 Param 1=&FloatFromCurve(FloatCurve=DataInterface() {propertyValues {Curve {keys [<at 0 0>,<at 1 1>]}}})
    Index 0 Param 2=1
    Index 0 Param 3=1
    Index 0 Param 4=1
    Module.Param0WriteEnabled=true
    Module.Param1WriteEnabled=true
    Module.Param2WriteEnabled=true
    Module.Param3WriteEnabled=true
   )
   &ScaleSpriteSize(Scale Factor=&Vector2DFromFloat(Value=&FloatFromCurve(FloatCurve=DataInterface() {propertyValues {Curve {keys [
          {arriveTangent=9.1531, interpMode=RCIM_Cubic, leaveTangent=9.1531, tangentMode=RCTM_User, time=0, value=0.02}
          {arriveTangent=0.48, interpMode=RCIM_Cubic, leaveTangent=0.48, time=0.18, value=0.99}
          <at 1 0.5 RCIM_Cubic>
         ]}}})))
   &PointAttractionForce(
    Attraction Radius=1250
    Attraction Strength=40
    Attractor Position:NiagaraPosition=$Particles.Initial.Position
    Kill Within Radius=false
    Module.Use Falloff=true
    UseAttractorOffset=false
   )
   &SolveForcesAndVelocity(Clamp Velocity=false, Limit Acceleration=false)
   &ScaleSpriteSizeBySpeed(Max Scale Factor=<1 4>, Min Scale Factor=<1 2>)
  ]
  rendererClasses [&NiagaraSpriteRendererProperties]
  renderers [{rendererClass=&NiagaraSpriteRendererProperties}]
 }
 Wings {
  emitterUpdateScript [
   &EmitterState(Life Cycle Mode=::NewEnumerator1, Loop Behavior=::NewEnumerator1, Loop Duration=1.6, UseLoopDelay=true)
   &SpawnBurst_Instantaneous(Spawn Count=2, Use Spawn Probability=false)
   &SpawnBurst_Instantaneous(Spawn Count=3, Spawn Time=0.2374, Use Spawn Probability=false, UseLoopCountLimit=false) {enabled=false, module=SpawnBurst_Instantaneous001}
  ]
  particleSpawnScript [
   /Niagara/Modules/Spawn/Initialization/InitializeParticle(
    Color=<rgba 1.7375 0 2 0.2>
    Mesh Scale=&Multiply_VectorByFloat(
     Vector=<1 0.8 1>
     Float=&SelectFloatFromArray(
      Direct Array Index=&ReturnExecIndex()
      Float Selection Array=DataInterface() {propertyValues {FloatData [0.8,0.8,0.3]}}
     )
    )
    Module.Write Color=false
    Module.Write Lifetime=true
    Module.Write Mass=false
    Module.Write Material Random=true
    Module.Write Position=false
    Module.Write Scale=true
    Module.Write SpriteRotation=false
    Module.Write SpriteSize=true
    Sprite Size=<8 8>
   ) {enabled=false}
   /Niagara/Modules/Spawn/Initialization/V2/InitializeParticle(
    Color=<rgba 2 1.5043 1.16 1>
    Lifetime=1.4
    Mesh Scale Mode=::NewEnumerator3
    Mesh Uniform Scale=0.1
    Sprite Size=<90 200>
    Sprite Size Mode=::NewEnumerator4
    UsePositionOffset=false
    Write Material Random=true
   ) {module=InitializeParticle001}
   &PartitionParticles()
   /SetVariables(Particles.Lifetime=&Add_Float(
     A=$Particles.Lifetime
     B=&SelectFloatFromArray(Direct Array Index=$Particles.PartitionParticles.Partition, Float Selection Array=DataInterface() {propertyValues {FloatData [-0.05,0.01]}})
    )) {isSetParameters=true}
   &InitialMeshOrientation(
    Mesh Orientation Mode=::NewEnumerator1
    Rotation=&SelectVectorFromArray(
     Direct Array Index=&ReturnExecIndex()
     Vector Selection Array=DataInterface() {propertyValues {FloatData [<0 0 0>, <0 0 0.5>]}}
    )
    Rotation Coordinate Space=::NewEnumerator2
    Use Rotation Vector=true
   ) {enabled=false}
   /SetVariables(Particles.Scale=&Multiply_Vector(
     A=$StackContext.Scale
     B=&SelectVectorFromArray(
      Direct Array Index=$Particles.PartitionParticles.Partition
      Vector Selection Array=DataInterface() {propertyValues {FloatData [<1 1 1>, <1 1 -1>]}}
     )
    )) {isSetParameters=true}
   /SetVariables(Particles.MaterialRandom:NiagaraFloat=&Multiply_FloatByInt(Float=0.75, Integer=$Particles.PartitionParticles.Partition)) {isSetParameters=true}
  ]
  particleUpdateScript [
   &ParticleState()
   /SetVariables(Particles.Fadeout:NiagaraFloat=&FloatFromCurve(
     FloatCurve=DataInterface() {propertyValues {Curve {keys [{tangentMode=RCTM_User, time=0, value=0}, {tangentMode=RCTM_User, time=1, value=1}]}}}
     CurveIndex=&RemapRange(Input Value To Remap=$Particles.NormalizedAge, Input Min=0.65, Input Max=1, Output Min=0, Output Max=1)
    )) {isSetParameters=true}
   &SkeletalMeshLocation(
    Bone / Socket Index=1
    Bone Sampling Mode=::NewEnumerator4
    MaskSpawnGroup=false
    Orientation Sampling=::NewEnumerator1
    Sampled Position Offset=&SelectVectorFromArray(
     Direct Array Index=&ReturnExecIndex()
     Vector Selection Array=DataInterface() {propertyValues {FloatData []}}
    )
    Skeletal Mesh=$User.SkeletalMesh
   ) {enabled=false, module=SkeletalMeshLocation001}
   /SetVariables(Particles.Position=&TransformPosition(Position=$System.SourcePosition, Source Space=::World, Destination Space=::Simulation)) {isSetParameters=true}
   /SetVariables(Particles.Position=&AddVectorToPosition(
     Position=$Particles.Position
     Vector=&TransformPosition(Position=<0 0 -5>, Source Space=::Local, Destination Space=::Simulation)
    )) {isSetParameters=true}
   /SetVariables(Particles.MeshOrientation=&MultiplyQuaternion(
     Quaternion A=&MakeQuatFromAxisAngle(
      Axis=&SelectVectorFromArray(
       Direct Array Index=$Particles.PartitionParticles.Partition
       Vector Selection Array=DataInterface() {propertyValues {FloatData [<0 1 0>, <0 -1 0>]}}
      )
      AngleInDegrees=&FloatFromCurve(
       FloatCurve=DataInterface() {propertyValues {Curve {keys [{arriveTangent=222.6932, interpMode=RCIM_Cubic, leaveTangent=222.6932, tangentMode=RCTM_User, time=0, value=-120}]}}}
       CurveIndex=$System.Timeline.NormalizedTime
      )
     )
     Quaternion B=<0 0 0 1>
    )) {isSetParameters=true}
   /SetVariables(Particles.LightCol:LinearColor=&MultiplyLinearColor(
     Linear Color=<rgba 0.5 0.089 0 1>
     Scale Factor=&MakeVector4_FromFloat(Float=&Sine(Normalized Angle=$Emitter.Age, Period=1, Scale=2, Bias=2))
    )) {isSetParameters=true}
   &ScaleColor(
    Scale Mode=::NewEnumerator1
    Scale RGBA=&MakeVector4_FromLinearColor(LinearColor=&ColorFromCurve(ColorCurve=DataInterface() {propertyValues {
        AlphaCurve {keys [<at 0 0.5>,<at 0.5 1>,<at 1 0.5>]}
        BlueCurve {keys [<at 0 1 RCIM_Cubic>,<at 0.5 2 RCIM_Cubic>,<at 1 1 RCIM_Cubic>]}
        GreenCurve {keys [<at 0 1 RCIM_Cubic>,<at 0.5 2 RCIM_Cubic>,<at 1 1 RCIM_Cubic>]}
        RedCurve {keys [<at 0 1 RCIM_Cubic>,<at 0.5 2 RCIM_Cubic>,<at 1 1 RCIM_Cubic>]}
       }}))
    ScaleRGBA=true
   ) {enabled=false}
   &DynamicMaterialParameters(
    Index 0 Param 1=&Add_Float(
     A=&FloatFromCurve(
      FloatCurve=DataInterface() {propertyValues {Curve {keys [{tangentMode=RCTM_User, time=0, value=1}, {tangentMode=RCTM_User, time=0.2, value=0}]}}}
      Scale Curve=2.5
     )
     B=&FloatFromCurve(
      FloatCurve=DataInterface() {propertyValues {Curve {keys [{interpMode=RCIM_Cubic, tangentMode=RCTM_User, time=0.35, value=0}, {interpMode=RCIM_Cubic, tangentMode=RCTM_User, time=1, value=1}]}}}
      CurveIndex=$Particles.Fadeout
      Scale Curve=2.5
     )
    )
    Index 0 Param 2=&FloatFromCurve(FloatCurve=DataInterface() {propertyValues {Curve {keys [<at 0 0 RCIM_Constant>,<at 0.5 1 RCIM_Constant>]}}})
    Index 0 Param 3=&FloatFromCurve(FloatCurve=DataInterface() {propertyValues {Curve {keys [{tangentMode=RCTM_User, time=0.85, value=0.5}, {tangentMode=RCTM_User, time=1, value=1}]}}})
    Index 1 Param 1=&Multiply_Float(
     A=&FloatFromCurve(
      FloatCurve=DataInterface() {propertyValues {Curve {keys [
          {arriveTangent=1.3294, interpMode=RCIM_Cubic, leaveTangent=1.3294, tangentMode=RCTM_User, time=0, value=-0.3}
          {arriveTangent=-0.1703, interpMode=RCIM_Cubic, leaveTangent=-0.1703, tangentMode=RCTM_User, time=0.6, value=0}
          <at 1 0 RCIM_Cubic>
         ]}}}
      CurveIndex=&Multiply_Float(A=$Particles.NormalizedAge, B=8)
      Scale Curve=2
     )
     B=0
    )
    Index 1 Param 2=&SelectFloatFromArray(Direct Array Index=$Particles.PartitionParticles.Partition, Float Selection Array=DataInterface() {propertyValues {FloatData [1,-1]}})
    Module.Param0WriteEnabled=true
    Module.Param0WriteEnabled1=true
    Module.Param1WriteEnabled=true
    Module.Param1WriteEnabled1=true
    Module.Param2WriteEnabled=true
    Module.Param2WriteEnabled1=true
    Module.Param3WriteEnabled=true
    Module.Param3WriteEnabled1=true
    Write Parameter Index 1=true
   )
   &ScaleColor(
    Scale Alpha=&Multiply_Float(
     A=1
     B=&FloatFromCurve(FloatCurve=DataInterface() {propertyValues {Curve {keys [<at 0 1>,<at 1 0>]}}}, CurveIndex=$Particles.Fadeout)
    )
    Scale RGB=&VectorFromFloat(Value=&Multiply_Float(
      A=1
      B=&FloatFromCurve(
       FloatCurve=DataInterface() {propertyValues {Curve {keys [
           {arriveTangent=9.0725, interpMode=RCIM_Cubic, leaveTangent=9.0725, tangentMode=RCTM_User, time=0, value=1}
           {arriveTangent=-2.5249, interpMode=RCIM_Cubic, leaveTangent=-2.5249, tangentMode=RCTM_User, time=0.1408, value=2}
           {arriveTangent=-0.0405, interpMode=RCIM_Cubic, leaveTangent=-0.0405, tangentMode=RCTM_User, time=1, value=1}
          ]}}}
       CurveIndex=&Multiply_Float(A=$Particles.NormalizedAge, B=2)
      )
     ))
    ScaleA=false
    ScaleRGB=true
   ) {module=ScaleColor001}
   &ScaleMeshSize(Scale Factor=&Multiply_VectorByFloat(
     Vector=&MakeVector(
      X=&FloatFromCurve(
       FloatCurve=DataInterface() {propertyValues {Curve {keys [
           {arriveTangent=4.1785, interpMode=RCIM_Cubic, leaveTangent=4.1785, tangentMode=RCTM_User, time=0, value=0}
           {arriveTangent=1, leaveTangent=1, time=0.3, value=1.15}
           {leaveTangent=-0.0405, time=1, value=1.1}
          ]}}}
       CurveIndex=&Multiply_Float(A=$Particles.NormalizedAge, B=1.5)
      )
      Y=1
      Z=1
     )
     Float=1
    ))
   &ScaleMeshSize(
    Initial Mesh Scale:Vector3f=$StackContext.Scale
    Scale Factor=&Lerp_Vector(
     A=<1 1 1>
     B=<1.2 1 1.2>
     Alpha=&FloatFromCurve(FloatCurve=DataInterface() {propertyValues {Curve {keys [<at 0 0 RCIM_Cubic>,<at 1 1 RCIM_Cubic>]}}}, CurveIndex=$Particles.Fadeout)
    )
   ) {module=ScaleMeshSize001}
   &CameraOffset(Camera Offset Amount=15)
   /SetVariables(
    Particles.GlowColor=&MultiplyLinearColor(
     Linear Color=&ColorFromCurve(ColorCurve=DataInterface() {propertyValues {
        AlphaCurve {keys [<at 0.0907 0>,<at 0.1804 1 RCIM_Cubic>]}
        BlueCurve {keys [<at 0.0963 0>]}
        GreenCurve {keys [<at 0.0963 0.2657>]}
        RedCurve {keys [<at 0.0963 2>]}
       }})
     Scale Factor=&MakeFloat4FromVectorAndFloat(
      "Vector (RGB)"=<1 1 1>
      "Float (Alpha)"=&FloatFromCurve(
       FloatCurve=DataInterface() {propertyValues {Curve {keys [{interpMode=RCIM_Cubic, tangentMode=RCTM_User, time=0, value=1}, {interpMode=RCIM_Cubic, tangentMode=RCTM_User, time=0.5, value=0}]}}}
       CurveIndex=$Particles.Fadeout
       Scale Curve=0.75
      )
     )
    )
    Particles.GlowScale:Vector3f=&Multiply_Vector(
     A=$Particles.Scale
     B=&MakeVector(
      X=&FloatFromCurve(FloatCurve=DataInterface() {propertyValues {Curve {keys [
           {arriveTangent=9.0725, time=0, value=1}
           {arriveTangent=-2.5249, leaveTangent=-2.5249, time=0.1408, value=1.1}
           <at 0.3 1.015>
           {arriveTangent=-0.0405, leaveTangent=-0.0405, time=1, value=1}
          ]}}})
      Y=1
      Z=1
     )
    )
   ) {isSetParameters=true}
   /SetVariables(
    Particles.BackingGlow:LinearColor=&MultiplyLinearColor(
     Linear Color=<rgba 1 0.1315 0 2>
     Scale Factor=&MakeFloat4FromVectorAndFloat(
      "Vector (RGB)"=&VectorFromFloat(Value=&Waveform("Waveform [1]"=::NewEnumerator12, "[1] Frequency"=7, "[1] Amplitude Min/Max"=<0.8 1.5>))
      "Float (Alpha)"=&FloatFromCurve(
       FloatCurve=DataInterface() {propertyValues {Curve {keys [
           {arriveTangent=9.0725, interpMode=RCIM_Cubic, time=0.1, value=0}
           {arriveTangent=1.8, interpMode=RCIM_Cubic, leaveTangent=1.8, time=0.1408, value=0.9847}
           {arriveTangent=-1.146, interpMode=RCIM_Cubic, leaveTangent=-1.146, time=0.6, value=0.9}
           {interpMode=RCIM_Cubic, leaveTangent=-0.0405, time=1, value=0}
          ]}}}
       Scale Curve=0.5
      )
     )
    )
    Particles.BackingGlowSize:Vector2f=&Multiply_Vector2D_ByFloat(
     Float=&Waveform("Waveform [1]"=::NewEnumerator12, "[1] Frequency"=5, "[1] Amplitude Min/Max"=<0.9 1.1>)
     Vector2D=<60.391 145.0103>
    )
   ) {isSetParameters=true}
   &SpriteFacingAndAlignment(
    Sprite Alignment=&RotateVectorByQuaternion(
     Vector To Rotate=<-1 0 0>
     Quaternion=&MultiplyQuaternion(
      Quaternion A=&MakeQuatFromAxisAngle(Axis=<0 1 0>, AngleInDegrees=90)
      Quaternion B=$StackContext.MeshOrientation
     )
    )
    WriteAlignment=true
    WriteFacing=false
   )
   /SetVariables(Particles.BlankDynParams:Vector4f=<1.5 2 0 0>) {isSetParameters=true}
   &FN_Light_Attributes(
    Color=$Particles.LightCol
    Create Color Attribute=true
    Create Position Attribute=true
    Position=&AddVectorToPosition(
     Position=&AddVectorToPosition(
      Position=$Particles.Position
      Vector=&RotateVectorByQuaternion(Vector To Rotate=<0 0 50>, Quaternion=$StackContext.MeshOrientation)
     )
     Vector=<0 0 40>
    )
    Scale Color=3
   )
  ]
  rendererClasses [
   &NiagaraMeshRendererProperties
   &NiagaraSpriteRendererProperties
   &NiagaraLightRendererProperties
  ]
  renderers [
   {rendererClass=&NiagaraMeshRendererProperties, rendererIndex=0}
   {rendererClass=&NiagaraSpriteRendererProperties, rendererIndex=1}
   {rendererClass=&NiagaraMeshRendererProperties, rendererIndex=2}
   {rendererClass=&NiagaraSpriteRendererProperties, rendererIndex=3}
   {rendererClass=&NiagaraLightRendererProperties, rendererIndex=4}
   {rendererClass=&NiagaraMeshRendererProperties, rendererIndex=5}
  ]
 }
}
summary {
 systemName=NS_Emote_CoatCheck
 emitters [
  {
   emitterName=Glow
   rendererClasses [
    &NiagaraSpriteRendererProperties
    &NiagaraLightRendererProperties
    &NiagaraMeshRendererProperties
    &NiagaraComponentRendererProperties
   ]
  }
  {bEnabled=false, emitterName=FN_AOE_Embers_Center, simTarget=GPUComputeSim, rendererClasses [&NiagaraSpriteRendererProperties]}
  {
   emitterName=Wings
   rendererClasses [
    &NiagaraMeshRendererProperties
    &NiagaraSpriteRendererProperties
    &NiagaraLightRendererProperties
   ]
  }
  {emitterName=Ring, rendererClasses [&NiagaraSpriteRendererProperties]}
  {emitterName=Sparks_Wings, simTarget=GPUComputeSim, rendererClasses [&NiagaraSpriteRendererProperties]}
  {bEnabled=false, emitterName=FN_Debug, rendererClasses [&NiagaraSpriteRendererProperties]}
 ]
 userVariables [
  User.bIsFrontEnd {type=NiagaraBool}
  User.bIsFrontEndPreview {type=NiagaraBool}
  User.Interpolate_Rate_of_Change {type=NiagaraFloat, defaultValue=0, description=""}
  User.Interpolate_WPO {type=NiagaraFloat, defaultValue=0, description=""}
  User.IsSkydiving {type=NiagaraBool}
  User.NormalizedNotifyProgress {type=NiagaraFloat, defaultValue=0, description=""}
  User.PlayerSpeed {type=NiagaraFloat, defaultValue=0, description=""}
  User.SkeletalMesh {
   type=NiagaraDataInterfaceSkeletalMesh
   defaultValue {
    struct=NiagaraExt_VariableValue_DataInterface
    value {dataInterface=&self:NiagaraDataInterfaceSkeletalMesh_0}
   }
  }
 ]
}
